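# Niagara system for the glow on the Alfredo cosmetic backpack, attached to a skeletal mesh and colored by a user emissive color.
# Glow burst-spawns soft-lifetime sprites at sampled mesh locations; Glow_Wings uses two spawn groups for bone-aligned sprites with mirrored offsets, plus light renderers.
# Both emitters flicker RGB through a shared-shape curve over system loop age and scale alpha by time of day, brighter during day.
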NiagaraSystem /BRCosmetics/Accessories/FORT_Backpacks/Backpack_M_MED_Alfredo/FX/NS_Backpack_Alfredo_Glow
dependencies {
 usedDataInterfaces [&NiagaraDataInterfaceMeshRendererInfo, &NiagaraDataInterfaceCurve]
 usedDynamicInputs []
 usedModules [&SystemState]
}
emitters {
 Glow {
  emitterUpdateScript [
   &EmitterState(
    Inactive Response=::NewEnumerator1
    Life Cycle Mode=::NewEnumerator1
    Loop Behavior=::NewEnumerator1
    Loop Duration Mode=::NewEnumerator1
    UseLoopDelay=false
   )
   &SpawnBurst_Instantaneous(Use Spawn Probability=false, UseLoopCountLimit=true)
  ]
  particleSpawnScript [
   &InitializeParticle(
    Color=$User.Emissive Color
    Position Offset=<100 0 0>
    Sprite Size Mode=::NewEnumerator3
    Uniform Sprite Size=18.2583
    UsePositionOffset=true
    Write Material Random=true
   )
   &SkeletalMeshLocation(MaskSpawnGroup=false, Mesh Sampling Type=::NewEnumerator5, Skeletal Mesh=$User.Skeletal Mesh, Socket Sampling Mode=::NewEnumerator5)
  ]
  particleUpdateScript [
   &ParticleState(Kill Particles When Lifetime Has Elapsed=false, Let Infinitely Lived Particles Die When Emitter Deactivates=true)
   &ScaleColor(
    Scale Alpha=&FNTimeOfDayScalar(Night Scale=1, Morning Scale=1, Day Scale=1.5, Evening Scale=1)
    Scale RGB=&VectorFromFloat(Value=&FloatFromCurve(
      FloatCurve=DataInterface() {propertyValues {Curve {keys [<at 0 1>,<at 0.0495 1.0491>,<at 0.0791 0.8714>,<at 0.1561 0.6964>,<at 0.222 1.0327>,<at 0.27 0.8868>,<at 0.3207 0.9188>,<at 0.3738 0.718>,<at 0.4401 0.8429>,<at 0.4803 1.0545>,<at 0.5384 0.9243>,<at 0.5897 0.7018>,<at 0.6556 0.7669>,<at 0.6931 0.9482>,<at 0.7538 1.0328>,<at 0.8218 0.8646>,<at 0.8909 0.8732>,<at 0.9368 0.756>,<at 1 1>]}}}
      CurveIndex=$System.NormalizedLoopAge
     ))
    ScaleA=true
    ScaleRGB=true
   ) {module=ScaleColor002}
   &ScaleColor(
    Scale Alpha=&FNTimeOfDayScalar(Night Scale=1, Morning Scale=1, Day Scale=0, Evening Scale=1)
    Scale RGB=&MakeVectorFromLinearColorRGB(Color=$User.Emissive Color)
    ScaleA=true
    ScaleRGB=true
   ) {enabled=false}
  ]
  rendererClasses [&NiagaraSpriteRendererProperties]
  renderers [{rendererClass=&NiagaraSpriteRendererProperties}]
 }
 Glow_Wings {
  emitterUpdateScript [
   &EmitterState(Life Cycle Mode=::NewEnumerator1, Loop Behavior=::NewEnumerator1, Loop Duration Mode=::NewEnumerator1, UseLoopDelay=false)
   &SpawnBurst_Instantaneous(Use Spawn Probability=false, UseLoopCountLimit=true)
   &SpawnBurst_Instantaneous(Spawn Group=1, Use Spawn Probability=false, UseLoopCountLimit=true) {module=SpawnBurst_Instantaneous001}
  ]
  particleSpawnScript [
   &FN_SetSpawnGroup(Number of Spawn Groups=2)
   &InitializeParticle(
    Color=$User.Emissive Color
    Lifetime=0
    Position Offset Coordinate Space=::Local
    Sprite Size=<22 70>
    Sprite Size Mode=::NewEnumerator4
    UsePositionOffset=true
    Write Material Random=true
   )
   &SkeletalMeshLocation(
    Bone / Socket Index=$Emitter.SpawnGroup
    Bone Sampling Mode=::NewEnumerator4
    MaskSpawnGroup=false
    Sampled Position Offset=&Lerp_Vector(A=<3 -4 -19>, B=<-3 -4 -19>, Alpha=&MakeFloatFromInt(Int=$Emitter.SpawnGroup))
    Sampled Position Offset Coordinate Space:ENiagaraSkelMeshTransforms=::NewEnumerator2
    Skeletal Mesh=$User.Skeletal Mesh
   )
   &SpriteFacingAndAlignment(
    Sprite Alignment=&RotateVectorByQuaternion(Vector To Rotate=<1 0 0>, Quaternion=$Particles.SkeletalMeshLocation.SampledRotation)
    WriteAlignment=true
    WriteFacing=false
   ) {module=SpriteFacingAndAlignment001}
  ]
  particleUpdateScript [
   &ParticleState(Kill Particles When Lifetime Has Elapsed=false, Let Infinitely Lived Particles Die When Emitter Deactivates=true)
   &FN_Light_Attributes(
    Color=$User.Emissive Color
    Create Color Attribute=true
    Create Exponent Attribute=true
    Create Position Attribute=true
    Create Radius Attribute=true
    Exponent=4
    Position:NiagaraPosition=$Particles.Position
    Radius=28
    Scale Color=11
   )
   &ScaleColor(
    Scale Alpha=&FNTimeOfDayScalar(Night Scale=1, Morning Scale=1, Day Scale=1.5, Evening Scale=1)
    Scale RGB=&VectorFromFloat(Value=&FloatFromCurve(
      FloatCurve=DataInterface() {propertyValues {Curve {keys [<at 0 1>,<at 0.0495 1.0491>,<at 0.0791 0.8714>,<at 0.1561 0.6964>,<at 0.222 1.0327>,<at 0.27 0.8868>,<at 0.3207 0.9188>,<at 0.3738 0.718>,<at 0.4401 0.8429>,<at 0.4803 1.0545>,<at 0.5384 0.9243>,<at 0.5897 0.7018>,<at 0.6556 0.7669>,<at 0.6931 0.9482>,<at 0.7538 1.0328>,<at 0.8218 0.8646>,<at 0.8909 0.8732>,<at 0.9368 0.756>,<at 1 1>]}}}
      CurveIndex=$System.NormalizedLoopAge
     ))
    ScaleA=true
    ScaleRGB=true
   ) {module=ScaleColor002}
  ]
  rendererClasses [&NiagaraSpriteRendererProperties, &NiagaraLightRendererProperties]
  renderers [
   {rendererClass=&NiagaraSpriteRendererProperties}
   {rendererClass=&NiagaraLightRendererProperties, rendererIndex=1}
   {rendererClass=&NiagaraLightRendererProperties, rendererIndex=2}
  ]
 }
}
summary {
 systemName=NS_Backpack_Alfredo_Glow
 emitters [
  {emitterName=Glow, rendererClasses [&NiagaraSpriteRendererProperties]}
  {emitterName=Glow_Wings, rendererClasses [&NiagaraSpriteRendererProperties, &NiagaraLightRendererProperties]}
 ]
 userVariables [
  User.bIsFrontEnd {type=NiagaraBool}
  User.bIsFrontEndPreview {type=NiagaraBool}
  User.Emissive Color {type=LinearColor, description="", defaultValue {struct=LinearColor, value=<rgba 1 1 1 1>}}
  User.Skeletal Mesh {
   type=NiagaraDataInterfaceSkeletalMesh
   defaultValue {
    struct=NiagaraExt_VariableValue_DataInterface
    value {dataInterface=&self:NiagaraDataInterfaceSkeletalMesh_1}
   }
  }
 ]
}
